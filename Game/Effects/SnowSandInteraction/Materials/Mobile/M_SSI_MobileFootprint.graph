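# Translucent unlit material for mobile snow and sand footprints on Niagara mesh particles.
# A constant feeds the emissive adjustment function, while shading is faked by comparing a transformed footprint normal map against the atmospheric light vector.
# That inverted lighting term is multiplied by particle alpha, the inverted MarkFlat dynamic parameter, and an inverted green-channel footprint mask, center-scaled by ScaleX and ScaleY.
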Material /Game/Effects/SnowSandInteraction/Materials/Mobile/M_SSI_MobileFootprint
graph {
 properties {
  bUsedWithNiagaraMeshParticles=true
  blendMode=BLEND_Translucent
  pixelDepthOffsetMode=PDOM_Legacy
  shadingModel=MSM_Unlit
  usedShadingModels=Unlit
 }
 nodes [
  MaterialFunctionCall(Emissive=Constant()) {materialFunction=&MF_FX_EmissiveAdjustments}
  n1 = TextureCoordinate()
  n2 = TextureSampleParameter2D(UVs=@n1) {parameterName=Normal, samplerType=SAMPLERTYPE_Normal, texture=/Game/Effects/SnowSandInteraction/Textures/SSI_T_Footprint_D}
  n3 = Abs(None=DotProduct(A=AtmosphericLightVector(), B=Transform(None=@n2.RGB)))
  n4 = OneMinus(None=DynamicParameter() {defaultValue=<rgba 1 1 0 0>, paramNames [CompressedQuaternion, BlurMarkScale, MarkFlat, MeshScale]}.MarkFlat)
  n5 = AppendVector(A=ScalarParameter() {defaultValue=1, parameterName=ScaleX}, B=ScalarParameter() {defaultValue=1, parameterName=ScaleY})
  n6 = OneMinus(None=TextureSample(UVs=MaterialFunctionCall(UVs=@n1, Texture Scale=@n5) {materialFunction=&ScaleUVsByCenter}.UVs) {texture=/Game/Effects/SnowSandInteraction/Textures/SSI_T_FootprintMasks}.G)
  Multiply(A=ParticleColor().A, B=Multiply(A=OneMinus(None=@n3), B=Multiply(A=@n4, B=@n6))) {constA=0.1}
 ]
}
